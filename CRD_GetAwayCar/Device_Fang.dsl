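; Device_Fang is a vehicle spawner device for the Fang vehicle in the GetAwayCar set, built on a spawner parent class.
; When spawn settings update, it sets the parent's vehicle type, then passes the enable-or-disable receiver flag on to the parent handler.
; It also converts the Color And Style user option into a zero-based integer by subtracting one, and its construction script defers to the parent.
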(event EventUpdateVehicleSpawnSettings
  (params (EnableOrDisableReceiver :type Boolean))
  ;[1 1 1] Set the appropriate VehicleType in Device_VehicleSpawnerParent_V2
  (Utilities|SetVehicleType
    (:OnVehicleTypeSet
      (|CallParentFunction EnableOrDisableReceiver))))

(fn ColorAndStyleAsAnInteger
  (return (- (Math|Conversions|ToInteger(Byte) (Variables|UserOptions|Basic|GetColorAndStyle)) 1)))

(fn ConstructionScript
  (|CallParentFunction))
